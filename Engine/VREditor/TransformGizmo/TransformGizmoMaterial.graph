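# Unlit material for the VR editor transform gizmo, tinted by a Color parameter and flagged for editor compositing.
# A vertically panning laser power texture adds a faint animated variation to the color.
# Fake directional shading comes from the pixel normal dotted with a constant light vector, biased, scaled and slightly lifted.
# The shaded result passes through EyeAdaptationInverse to counter auto exposure.
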Material /Engine/VREditor/TransformGizmo/TransformGizmoMaterial
graph {
 properties {bUsedWithEditorCompositing=true, pixelDepthOffsetMode=PDOM_Legacy, shadingModel=MSM_Unlit, usedShadingModels=Unlit}
 nodes [
  n1 = Panner(Coordinate=TextureCoordinate() {vTiling=0}) {speedY=-0.2}
  n2 = VectorParameter() {parameterName=Color, defaultValue=<rgba 0.4 0.4 0.4 1>}
  n3 = Multiply(A=Multiply(A=TextureSample(UVs=@n1) {texture=/Engine/VREditor/LaserPointer/VR_LaserPower_01}.RGB) {constB=0.1}, B=@n2.RGB)
  n4 = Multiply(A=Constant3Vector() {desc=Fake light direction, constant=<rgba 1 1.5 -3 0>}) {constB=-1}
  n5 = ConstantBiasScale(None=DotProduct(A=Normalize(VectorInput=@n4), B=PixelNormalWS()))
  EyeAdaptationInverse(LightValueInput=Multiply(A=Add(A=@n3, B=@n2.RGB), B=Add(A=@n5) {constB=0.2}), AlphaInput=Constant() {r=1})
 ]
}
