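# Translucent unlit material for the VR Editor transform gizmo, drawn without depth testing so it shows through scene geometry.
# It fakes directional lighting by dotting a constant light direction with the pixel normal, remapping and lifting the result, then tinting it by the Color parameter.
# EyeAdaptationInverse counters auto-exposure on that shaded color, and a scalar Opacity parameter provides the translucency control.
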Material /Engine/VREditor/TransformGizmo/TranslucentTransformGizmoMaterial
graph {
 properties {
  bDisableDepthTest=true
  bUsedWithEditorCompositing=true
  blendMode=BLEND_Translucent
  pixelDepthOffsetMode=PDOM_Legacy
  shadingModel=MSM_Unlit
  usedShadingModels=Unlit
 }
 nodes [
  n1 = Multiply(A=Constant3Vector() {desc=Fake light direction, constant=<rgba 1 1.5 -3 0>}) {constB=-1}
  n2 = ConstantBiasScale(None=DotProduct(A=Normalize(VectorInput=@n1), B=PixelNormalWS()))
  EyeAdaptationInverse(
   LightValueInput=Multiply(A=VectorParameter() {parameterName=Color, defaultValue=<rgba 0.4 0.4 0.4 1>}.RGB, B=Add(A=@n2) {constB=0.2})
   AlphaInput=Constant() {r=1}
  )
  ScalarParameter() {defaultValue=0.6, parameterName=Opacity}
 ]
}
