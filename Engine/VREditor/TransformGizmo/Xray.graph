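# Xray is a material in the VR Editor's TransformGizmo folder, presumably for drawing an X-ray look on the transform gizmo.
# It uses the Default Lit shading model with legacy pixel depth offset mode.
# Its graph holds a single vector parameter with a muted green default color and zero alpha, with no connection to any material input shown.
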Material /Engine/VREditor/TransformGizmo/Xray
graph {
 properties {pixelDepthOffsetMode=PDOM_Legacy, usedShadingModels=Default Lit}
 nodes [VectorParameter() {defaultValue=<rgba 0.3843 0.6314 0.4941 0>}]
}
